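; PCG graph template for specific biome data that prepares root points to send to the Local Biome Core.
; It creates a point at the biome data actor in original component space and uses its position as a world ray hit origin.
; The ProjectPrimary parameter, on by default, branches the point to either snap onto the hit surface without projecting rotation or pass straight to the output.
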(graph SpecificBiomeData_Template
  (params (ProjectPrimary :type Bool :default true))
  ;[0 0 0 1 24] 1-Prepare root point(s) to send to the Local Biome Core
  ;;[0 0 0 1 12] i.e Create a point on the specific biome data actor and project it onto the World Ray Hit surface to snap it down to the nearest surface.
  (bind CreatePoints_1 (Create%20Points :CoordinateSpace OriginalComponent :PointsToCreate [(struct)]))
  (bind Branch_6 (Branch :bOutputToB-param ProjectPrimary))
  (bind GetAttributeFromPointIndex_13 (Get%20Attribute%20From%20Point%20Index :bForceOutputAttributeToBeInElementsDomain true :InputSource $Position :OutputAttributeName None))
  (bind WorldRayHitQuery_12 (World%20Ray%20Hit%20Query :QueryParams (struct :bOverrideDefaultParams true :RayLength 1e+03)))
  (bind Projection_10 (Projection :ProjectionParams (struct :bProjectRotations false)))
  (connect CreatePoints_1 Out Branch_6 In)
  (connect CreatePoints_1 Out GetAttributeFromPointIndex_13 In)
  (connect Branch_6 Output%20B Projection_10 In)
  (connect WorldRayHitQuery_12 Out Projection_10 Projection%20Target)
  (connect GetAttributeFromPointIndex_13 Attribute WorldRayHitQuery_12 RayOrigin)
  (output Out Branch_6.Output%20A)
  (output Out Projection_10))
