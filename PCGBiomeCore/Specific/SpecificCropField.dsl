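; PCG Biome Core graph that generates crop-field placement points from tagged splines.
; It samples SplineStart and SplineEnd equally, jitters the points, then partitions them by copied index to build one spline per row.
; Resampled column points are jittered, clipped to the SplineShape interior, projected onto landscape or world collision, and randomly rotated and scaled.
; Debug branches preview each stage; declared Asset, Mesh and spawner parameters are unused here.
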(graph SpecificCropField
  (params (Asset :type (Object /PCGBiomeCore/BiomeAssets/Setup/BiomeAssetBaseTemplate.BiomeAssetBaseTemplate_C)) (RowsCount :type Int :default 10) (ColumnsCount :type Int :default 20) (RowsNoise :type Double :default 1e+02) (ColumnsNoise :type Double :default 5e+01) (ProjectOnLandscapeOnly :type Bool) (UseInternalStaticMeshSpawner :type Bool) (Mesh :type (SoftObject /Script/Engine.StaticMesh)) (DebugRows :type Bool) (DebugColumns :type Bool) (DebugPlacement :type Bool))
  (bind GetSplineData_83 (Get%20Spline%20Data :bComponentsMustOverlapSelf false))
  (bind Merge_12 (Merge%20Points))
  (bind BoundsModifier_15 (Bounds%20Modifier :BoundsMax (2e+01 2e+01 2e+01) :BoundsMin (-2e+01 -2e+01 -2e+01) :Mode Set))
  (bind FilterDataByTag_1 (Filter%20Data%20By%20Tag :bTokenizeOnWhiteSpace true :SelectedTags "BoundingShape"))
  (bind DeleteTags_0 (Delete%20Tags :Operation KeepOnlySelectedTags))
  ;[0 0 0] Add random offsets
  (bind MakeVectorAttribute_21 (Make%20Vector%20Attribute :OutputType Vector :X RowsNoise :Y RowsNoise))
  (bind CreateAttribute_20 (Create%20Constant :AttributeTypes (struct :DoubleValue -1.0)))
  (bind AttributeMathsOp_19 (Attribute%20Maths%20Op :Operation Multiply :InA RowsNoise))
  (bind MakeVectorAttribute_23 (Make%20Vector%20Attribute :OutputType Vector))
  (bind TransformPoints_17 (Transform%20Points))
  ;[0 0 0] Get Spline from BP by Component Tag
  (bind FilterByTag_1 (Filter%20Data%20By%20Tag :bTokenizeOnWhiteSpace true :SelectedTags "SplineStart,SplineEnd"))
  ;[0 0 0] Sample both splines with same point count
  (bind SplineSampler_1 (Spline%20Sampler :SamplerParams (struct :SubdivisionsPerSegment 10) :SubdivisionsPerSegment RowsCount))
  ;[0 0 0] Copy point index to point attribute
  (bind CopyAttribute_9 (Copy%20Attributes :InputSource $Index :OutputTarget index_attr))
  ;[0 0 0] Partition data into multiple data sets of Start/End points using  index_attr point attribute
  (bind AttributePartition_8 (Attribute%20Partition :bTokenizeOnWhiteSpace true :PartitionAttributeSelectors [index_attr]))
  ;[0 0 0] Create a spline for each partition
  (bind CreateSpline_0 (Create%20Spline :ArriveTangentAttribute None :InterpTypeAttribute None :LeaveTangentAttribute None))
  ;[0 0 0] Sample created splines for columns
  (bind SplineSampler_7 (Spline%20Sampler :SamplerParams (struct :SubdivisionsPerSegment 3) :SubdivisionsPerSegment ColumnsCount))
  ;[0 0 0] Merge all (partitioned) spline points to a single data set
  (bind Merge_9 (Merge%20Points))
  ;[0 0 0] Add random columns noise offset
  (bind MakeVectorAttribute_31 (Make%20Vector%20Attribute :OutputType Vector :X ColumnsNoise :Y ColumnsNoise))
  (bind CreateAttribute_29 (Create%20Constant :AttributeTypes (struct :DoubleValue -1.0)))
  (bind AttributeMathsOp_27 (Attribute%20Maths%20Op :Operation Multiply :InA ColumnsNoise))
  (bind MakeVectorAttribute_33 (Make%20Vector%20Attribute :OutputType Vector))
  (bind TransformPoints_25 (Transform%20Points))
  ;[0 0 0] Intersect with 3rd spline
  (bind SplineSampler_46 (Spline%20Sampler :SamplerParams (struct :Dimension OnInterior :SubdivisionsPerSegment 3)))
  (bind BoundsModifier_74 (Bounds%20Modifier :BoundsMax (1.0 1.0 5e+01) :BoundsMin (1.0 1.0 5e+01)))
  (bind Intersection_0 (Intersection))
  (bind ToPoint_17 (To%20Point))
  ;;[0 0 0] Get Spline from BP by Component Tag
  (bind GetSplineData_97 (Get%20Spline%20Data :bComponentsMustOverlapSelf false))
  (bind FilterByTag_99 (Filter%20Data%20By%20Tag :bTokenizeOnWhiteSpace true :SelectedTags "SplineShape"))
  ;[0 0 0] Project Points on World Collisions or Landscape
  (bind GetLandscapeData_49 (Get%20Landscape%20Data :bComponentsMustOverlapSelf false :SamplingProperties (struct :bGetLayerWeights false)))
  (bind Branch_16 (Branch :bOutputToB-param ProjectOnLandscapeOnly))
  (bind WorldRayHitQuery_0 (World%20Ray%20Hit%20Query :QueryParams (struct :bTraceComplex true)))
  (bind Projection_77 (Projection))
  (bind Projection_19 (Projection))
  ;[1 0.3333 0] Debug Rows
  (bind Branch_3 (Branch :bOutputToB-param DebugRows))
  (bind Debug_65 (Debug))
  ;[1 0.3333 0] Debug Columns
  (bind Branch_6 (Branch :bOutputToB-param DebugColumns))
  (bind Debug_69 (Debug))
  ;[0 0 0] Random rotation/scale
  (bind TransformPoints_101 (Transform%20Points :bAbsoluteScale true :RotationMax (struct :Pitch 1e+01 :Yaw 1.8e+02 :Roll 1e+01) :RotationMin (struct :Pitch -1e+01 :Yaw -1.8e+02 :Roll -1e+01) :ScaleMax (1.0 1.0 1.5) :ScaleMin (1.0 1.0 0.5)))
  ;[1 0.3333 0] Debug Columns
  (bind Branch_10 (Branch :bOutputToB-param DebugPlacement))
  (bind Debug_19 (Debug))
  (reroute VolumeBoundingShape FilterDataByTag_1.InsideFilter)
  (connect CreateAttribute_20 Out AttributeMathsOp_19 InB)
  (connect CreateAttribute_29 Out AttributeMathsOp_27 InB)
  (connect TransformPoints_17 Out AttributePartition_8 In)
  (connect Merge_9 Out BoundsModifier_15 In)
  (connect SplineSampler_46 Out BoundsModifier_74 In)
  (connect SplineSampler_1 Out Branch_3 In)
  (connect BoundsModifier_15 Out Branch_6 In)
  (connect TransformPoints_101 Out Branch_10 In)
  (connect ToPoint_17 Out Branch_16 In)
  (connect SplineSampler_1 Out CopyAttribute_9 Source)
  (connect SplineSampler_1 Out CopyAttribute_9 Target)
  (connect AttributePartition_8 Out CreateSpline_0 In)
  (connect Branch_10 Output%20B Debug_19 In)
  (connect Branch_3 Output%20B Debug_65 In)
  (connect Branch_6 Output%20B Debug_69 In)
  (connect TransformPoints_101 Out DeleteTags_0 In)
  (connect GetSplineData_83 Out FilterByTag_1 In)
  (connect GetSplineData_97 Out FilterByTag_99 In)
  (connect DefaultInputNode In FilterDataByTag_1 In)
  (connect TransformPoints_25 Out Intersection_0 Primary%20Source)
  (connect BoundsModifier_74 Out Intersection_0 Source1)
  (connect AttributeMathsOp_19 Out MakeVectorAttribute_23 X)
  (connect AttributeMathsOp_19 Out MakeVectorAttribute_23 Y)
  (connect AttributeMathsOp_27 Out MakeVectorAttribute_33 X)
  (connect AttributeMathsOp_27 Out MakeVectorAttribute_33 Y)
  (connect SplineSampler_7 Out Merge_9 In)
  (connect CopyAttribute_9 Out Merge_12 In)
  (connect Branch_16 Output%20B Projection_19 In)
  (connect GetLandscapeData_49 Out Projection_19 Projection%20Target)
  (connect Branch_16 Output%20A Projection_77 In)
  (connect WorldRayHitQuery_0 Out Projection_77 Projection%20Target)
  (connect VolumeBoundingShape Out SplineSampler_1 Bounding%20Shape)
  (connect FilterByTag_1 InsideFilter SplineSampler_1 Spline)
  (connect VolumeBoundingShape Out SplineSampler_7 Bounding%20Shape)
  (connect CreateSpline_0 Out SplineSampler_7 Spline)
  (connect VolumeBoundingShape Out SplineSampler_46 Bounding%20Shape)
  (connect FilterByTag_99 InsideFilter SplineSampler_46 Spline)
  (connect Intersection_0 Out ToPoint_17 In)
  (connect Merge_12 Out TransformPoints_17 In)
  (connect MakeVectorAttribute_23 Out TransformPoints_17 OffsetMax)
  (connect MakeVectorAttribute_21 Out TransformPoints_17 OffsetMin)
  (connect BoundsModifier_15 Out TransformPoints_25 In)
  (connect MakeVectorAttribute_33 Out TransformPoints_25 OffsetMax)
  (connect MakeVectorAttribute_31 Out TransformPoints_25 OffsetMin)
  (connect Projection_19 Out TransformPoints_101 In)
  (connect Projection_77 Out TransformPoints_101 In)
  (output Out DeleteTags_0))
